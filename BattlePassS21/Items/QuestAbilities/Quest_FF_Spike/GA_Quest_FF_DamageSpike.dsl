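; Passive quest ability for the FF Spike quest that detects bursts of damage dealt by the player.
; A gameplay event task listens for Event.Damage.Dealt; the first hit starts a ten-second timer, and event magnitudes accumulate until it expires and clears the total.
; Each hit during an active window that leaves the total above 10,000 sends the Creative.Quest.FF.Spike stat event to the owning player's quest manager.
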(event Ability|EventOnEndAbility
  (params (bWasCancelled :type Boolean))
  (Ability|EndAbility))

(event Ability|EventFailedToActivatePassiveAbility
  (params (ActorInfo :type Gameplay%20Ability%20Actor%20Info%20Structure))
  (Ability|EndAbility))

(event Ability|EventActivateAbility
  (bind _damage_dealt (Variables|Default|GetDamage%20Dealt))
  (|CallParentFunction)
  (bind (_asynctaskproxy _payload) (Ability|Tasks|WaitGameplayEvent "(TagName=\"Event.Damage.Dealt\")")
    (:EventReceived
      (if (Variables|Default|GetTimerActive)
        (bind _output_get (Variables|Default|SetDamage%20Dealt (+ _damage_dealt (.EventMagnitude (Utilities|Struct|BreakGameplayEventData _payload)))))
        (if (> _output_get 1e+04)
          (Quest|SendComplexCustomStatEvent (Variables|GAVars|Data|GetOwningPlayerQuestManager) 0 "(GameplayTags=((TagName=\"Creative.Quest.FF.Spike\")))" "(GameplayTags=)" 1))
        (else
          (Variables|Default|SetDamage%20Dealt (+ _damage_dealt (.EventMagnitude (Utilities|Struct|BreakGameplayEventData _payload))))
          (Utilities|Time|SetTimerbyEvent (delegate Custom|TimerUp) 1e+01)
          (Variables|Default|SetTimerActive true))))) ; Start counting damage when player deals damage
)

(event Custom|TimerUp
  (Variables|Default|SetTimerActive false)
  (Variables|Default|SetDamage%20Dealt 0.0))
